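# Material function MF_Quantize, in the Turtleneck Male pickaxe FX folder, remaps a scalar gradient into a discrete number of bands.
# It multiplies the gradient by Steps (preview 5), floors it, then divides by Steps to give the Default Range output.
# Increased Range rescales that result by 1 + 1/(Steps - 1), intended to make the first band 0 and the last 1.
# The gradient previews with the U texture coordinate.
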MaterialFunction /BRCosmetics/Weapons/FORT_Melee/Pickaxe_Turtleneck_Male/FX/Default/MF_Quantize
graph {
 properties {description=Remaps the input gradient into a discrete amount of steps, libraryCategoriesText [- TuataraVFX]}
 nodes [
  n1 = FunctionInput(Preview=ComponentMask(None=TextureCoordinate()) {r=true}) {bUsePreviewValueAsDefault=true, inputName=Gradient, inputType=FunctionInput_Scalar}
  n2 = FunctionInput() {
   bUsePreviewValueAsDefault=true
   description=Number of steps
   inputName=Steps
   inputType=FunctionInput_Scalar
   sortPriority=1
   previewValue=<5 0 0 1>
  }
  n3 = Divide(A=Floor(None=Multiply(A=@n1, B=@n2)), B=@n2)
  FunctionOutput(None=@n3) {description=Default range, outputName=Default Range}
  n4 = Divide(B=Subtract(A=@n2))
  FunctionOutput(None=Multiply(A=Add(A=@n4), B=@n3)) {description="Slightly larger range, making sure the first band = 0 and the last = 1", outputName=Increased Range, sortPriority=1}
 ]
}
