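# Unlit, two-sided translucent material for the Turtleneck pickaxe blade effect, flagged for Niagara mesh particles.
# UVs mirrored around the V midline are warped by panning noise and drive further noise layers.
# Those layers are eroded by distance from the midline and quantized into stepped bands, then mapped through red-default color ramps and highlights.
# Branches apply the texture alpha mask, time-of-day tint, view-angle fade, vertex color and optional particle alpha.
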Material /BRCosmetics/Weapons/FORT_Melee/Pickaxe_Turtleneck_Male/FX/Default/M_Turtleneck_Blade
graph {
 properties {
  bUsedWithNiagaraMeshParticles=true
  blendMode=BLEND_Translucent
  pixelDepthOffsetMode=PDOM_Legacy
  shadingModel=MSM_Unlit
  twoSided=true
  usedShadingModels=Unlit
 }
 nodes [
  n1 = VectorParameter() {parameterName=Morning, defaultValue=<rgba 0.15 1 1 1>}
  n2 = VectorParameter() {parameterName=Day, defaultValue=<rgba 1 1 1 1>}
  n3 = VectorParameter() {parameterName=Evening, defaultValue=<rgba 0.25 1 1 1>}
  n4 = VectorParameter() {parameterName=Night, defaultValue=<rgba 1 1 1 1>}
  n5 = TextureSampleParameter2D() {parameterName=Alpha Tex, texture=&/T_FX_Turtleneck}
  n6 = VectorParameter() {parameterName=ColorF}
  n7 = MaterialFunctionCall(Float2=TextureCoordinate()) {materialFunction=&BreakOutFloat2Components}
  n8 = Subtract(A=@n7.G) {constB=0.5}
  n9 = Abs(None=@n8)
  n10 = Add(A=AppendVector(A=@n7.R, B=OneMinus(None=@n9)), B=Divide(A=Sign(None=@n8)) {constB=3})
  n11 = AppendVector(
   A=ScalarParameter() {defaultValue=0.5, group=Distortion_Ctrl, parameterName=T_Distortion_Sub_U_Tile}
   B=ScalarParameter() {defaultValue=0.25, group=Distortion_Ctrl, parameterName=T_Distortion_Sub_V_Tile}
  )
  n12 = ScalarParameter() {defaultValue=0.75, group=Distortion_Ctrl, parameterName=T_Distortion_Sub_U_Speed}
  n13 = ScalarParameter() {defaultValue=-0.25, group=Distortion_Ctrl, parameterName=T_Distortion_Sub_V_Speed}
  n14 = TextureSample(UVs=Panner(Coordinate=Multiply(A=@n10, B=@n11), Speed=AppendVector(A=@n12, B=@n13))) {
   samplerType=SAMPLERTYPE_LinearGrayscale
   texture=&/T_Pickaxe_Turtleneck_Noise
  }
  n15 = Add(
   A=@n10
   B=Multiply(A=@n14.R, B=ScalarParameter() {defaultValue=0.1, group=Distortion_Ctrl, parameterName=Distortion_Sub_Intensity}) {constB=0.1}
  )
  n16 = VectorParameter() {parameterName=Speeds, defaultValue=<rgba 1.5 -0.5 1.25 -0.25>}
  n17 = TextureSample(UVs=Panner(Coordinate=@n15, Speed=AppendVector(A=@n16.R, B=@n16.R))) {
   samplerType=SAMPLERTYPE_LinearGrayscale
   texture=&T_Pickaxe_Turtleneck_Noise
  }
  n18 = Multiply(A=@n9) {constB=2}
  n19 = SquareRoot(None=@n18)
  n20 = MaterialFunctionCall!(Gradient=@n17.R, Erosion Amount=@n19, Softness=ScalarParameter() {defaultValue=0.85, parameterName=Softness}) {materialFunction=&/MF_Erosion}
  n21 = Divide(A=MaterialFunctionCall!(Gradient=@n20.Increased Range, Steps) {materialFunction=&/MF_Quantize}.Increased Range)
  n22 = ScalarParameter() {parameterName=Dark_V}
  n23 = OneMinus(None=@n18)
  n24 = VectorParameter() {parameterName=Color C, defaultValue=<rgba 0.5 0 0.025 0>}
  n25 = TextureSample(UVs=Panner() {speedX=1}) {texture=&T_FX_Turtleneck}
  n26 = VectorParameter() {parameterName=Speeds02, defaultValue=<rgba 1.25 -0.35 1.25 -0.25>}
  n27 = AppendVector(A=@n26.RGB, B=@n26.RGB)
  n28 = TextureSample(UVs=Multiply(A=Panner(Coordinate=@n15, Speed=@n27)) {constB=2}) {
   samplerType=SAMPLERTYPE_LinearGrayscale
   texture=&T_Pickaxe_Turtleneck_Noise
  }
  n29 = Multiply(A=Multiply(A=@n25.B, B=@n28.R), B=@n20.Increased Range)
  n30 = AppendVector(A=@n16.B, B=@n16.B)
  n31 = TextureSample(UVs=Multiply(A=Panner(Coordinate=@n15, Speed=@n30)) {constB=2.25}) {
   samplerType=SAMPLERTYPE_LinearGrayscale
   texture=&T_Pickaxe_Turtleneck_Noise
  }
  n32 = MaterialFunctionCall(Gradient=@n31.R, Erosion Amount=@n19, Softness=ScalarParameter() {defaultValue=0.8, parameterName=Softness02}) {materialFunction=&MF_Erosion}
  n33 = Add(
   A=MaterialFunctionCall(Gradient=@n20.Increased Range, Steps=ScalarParameter() {defaultValue=1, parameterName=Steps}) {materialFunction=&MF_Quantize}.Increased Range
   B=Divide(A=MaterialFunctionCall(Gradient=@n32.Increased Range, Steps=@n25.R) {materialFunction=&MF_Quantize}.Increased Range) {constB=3}
  )
  n34 = Add(A=Divide(A=Multiply(A=@n29, B=@n25.R)) {constB=5}, B=Saturate(None=Divide(A=@n33) {constB=1}))
  n35 = Divide(A=Add(A=@n25.G, B=@n34)) {constB=2.5}
  n36 = ScalarParameter() {defaultValue=1, parameterName=Steps}
  n37 = MaterialFunctionCall(
   Gradient=MaterialFunctionCall(Gradient=@n28.R, Erosion Amount=@n19, Softness=ScalarParameter() {defaultValue=1.5, parameterName=Softness03}) {materialFunction=&MF_Erosion}.Increased Range
   Steps=@n36
  ) {materialFunction=&MF_Quantize}
  n38 = Add(
   A=Multiply(A=@n23, B=VectorParameter() {parameterName=ColorD, defaultValue=<rgba 0.35 0 0.025 0>}.RGB)
   B=MaterialFunctionCall(
    1=ScalarParameter()
    2=@n24.RGB
    3=VectorParameter() {parameterName=Color B, defaultValue=<rgba 0.9 0 0.025 0>}.RGB
    4=VectorParameter() {parameterName=Color A, defaultValue=<rgba 15 0 0.25 0>}.RGB
    A=Add(A=@n35, B=@n37.Default Range)
   ) {materialFunction=&Lerp_Multiple_Float3}.Lerp 4 Inputs
  )
  n39 = Add(A=Multiply(A=Multiply(A=@n6.RGB, B=@n21), B=@n22), B=Add(A=@n38, B=Multiply(B=@n24.RGB)))
  MaterialFunctionCall(Emissive=Multiply(
    A=MaterialFunctionCall(Value_Morning=@n1.RGB, Value_Day=@n2.RGB, Value_Evening=@n3.RGB, Value_Night=@n4.RGB) {materialFunction=&Scalar_Time_of_Day}.Result
    B=Multiply(A=@n5.A, B=@n39)
   )) {materialFunction=&MF_FX_EmissiveAdjustments}
  n40 = Power(Base=@n34, Exp=ScalarParameter() {defaultValue=1, parameterName=Power_Dark_Edge}) {constExponent=0.5}
  n41 = VectorParameter() {parameterName=ColorHighlight, defaultValue=<rgba 10 0 0.25 0>}
  n42 = ScalarParameter() {defaultValue=1.5, parameterName=Softness05}
  n43 = Clamp(None=MaterialFunctionCall(
    Gradient=MaterialFunctionCall(Gradient=@n31.R, Erosion Amount=@n19, Softness=@n42) {materialFunction=&MF_Erosion}.Increased Range
    Steps=@n36
   ) {materialFunction=&MF_Quantize}.Increased Range)
  n44 = Multiply(A=Add(A=Add(A=@n40, B=@n38), B=Multiply(A=@n41.RGB, B=@n43)), B=@n22)
  n45 = Multiply(A=Divide(A=@n23) {constB=1}, B=Multiply(A=@n5.A, B=Add(A=@n23, B=@n44)))
  n46 = MaterialFunctionCall("FadeSensitivity (optional)"=Constant() {r=1}) {materialFunction=&FadeBasedOnViewAngle}
  Multiply(
   A=Multiply(A=Multiply(A=@n45, B=@n46.Result), B=VertexColor())
   B=LinearInterpolate(B=ParticleColor().A, Alpha=ScalarParameter() {parameterName=useParticleAlpha}) {constA=1}
  )
  OneMinus(None=@n5.A)
 ]
}
